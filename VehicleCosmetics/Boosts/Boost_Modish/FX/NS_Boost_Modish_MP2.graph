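# Niagara system for the Modish vehicle boost cosmetic, emitting ribbon trails and sprite and mesh effects from the car's skeletal mesh sockets.
# One emitter drives three ribbon renderers, with offset and velocity chosen by game mode ID.
# Three sprite emitters partition particles across ports and use spawn groups to vary layers, including a looping-lifetime emitter with a spinning mesh.
# Painted-variant overrides supply per-color tints to every emitter, animated by curve-driven color, size and material parameters.
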NiagaraSystem /VehicleCosmetics/Boosts/Boost_Modish/FX/NS_Boost_Modish_MP2
dependencies {
 usedDataInterfaces [
  &NiagaraDataInterfaceCurve
  &NiagaraDataInterfaceArrayFloat
  &NiagaraDataInterfaceArrayFloat3
  &NiagaraDataInterfaceSkeletalMesh
 ]
 usedDynamicInputs []
 usedModules [
  &self:SystemSpawnScript.SystemScriptSource.SystemScriptGraph.NiagaraNodeAssignment_0.SetVariables_#1
  &SystemState
 ]
}
emitters {
 NE_VC_Boost_Ribbon_Parent001 {
  emitterUpdateScript [
   /SetVariables(Emitter.IsDefColor=HlslExpression() {hlslExpression="(User.ColorVariantID) == 0 ? true : false"}) {isSetParameters=true}
   &EmitterState()
   &VC_FilteredSocketCount()
   &VC_RateQualityHandler(High Spawn Rate=60, Low Spawn Rate=60, Medium Spawn Rate=60, Use Custom Rate?=true)
   &VC_SpawnRibbon_v1(Number of Ribbons=$Emitter.PortCount, SpawnRate=$Emitter.Spawn Rate)
   /SetVariables(Emitter.DefColorCheck:NiagaraBool=&Boolean_NotOperation(Bool=$Emitter.IsDefColor)) {isSetParameters=true}
  ]
  particleSpawnScript [
   /SetVariables(Particles.OutlineRWidth=40) {isSetParameters=true}
   &VC_PaintedVariantsOverride(
    Black=<rgba 0 0 0 1>
    BurntSienna=<rgba 0.0642 0.0198 0.0079 1>
    DefaultColor=<rgba 0.5 0.5 0.5 1>
    ForestGreen=<rgba 0 0.5833 0.0295 1>
    Lime=<rgba 0.2943 0.5885 0 1>
    Pink=<rgba 0.5052 0.0605 0.3435 1>
    R1 Black:LinearColor=<rgba 0.0208 0.0208 0.0208 1>
    R1 BurntSienna:LinearColor=<rgba 0.0469 0.0145 0.0058 1>
    R1 Cobalt:LinearColor=<rgba 0 0.0185 1 1>
    R1 Crimson:LinearColor=<rgba 0.1146 0 0.0007 1>
    R1 DefaultColor=<rgba 3 1.2214 0 1>
    R1 ForestGreen:LinearColor=<rgba 0.0027 0.151 0 2>
    R1 Grey:LinearColor=<rgba 0.0625 0.0625 0.0625 1>
    R1 Lime:LinearColor=<rgba 0.2609 0.3438 0 1>
    R1 Orange:LinearColor=<rgba 0.1458 0.0753 0 1>
    R1 Pink:LinearColor=<rgba 0.8281 0.0992 0.5631 1>
    R1 Purple:LinearColor=<rgba 0.0654 0 0.3229 1>
    R1 Saffron:LinearColor=<rgba 1.1979 0.9184 0 1>
    R1 SkyBlue:LinearColor=<rgba 0 0.253 0.3802 1>
    R1 TitaniumWhite:LinearColor=<rgba 0.2917 0.2917 0.2917 1>
    R2 Black:LinearColor=<rgba 0.1198 0.1198 0.1198 1>
    R2 BurntSienna:LinearColor=<rgba 0.2865 0.0688 0.0104 1>
    R2 Cobalt:LinearColor=<rgba 0.0627 0.1495 0.8958 1>
    R2 Crimson:LinearColor=<rgba 0.4427 0 0.012 1>
    R2 ForestGreen:LinearColor=<rgba 0 0.3229 0.0163 1>
    R2 Grey:LinearColor=<rgba 0.4 0.4 0.4 1>
    R2 Lime:LinearColor=<rgba 0.7066 1.25 0 1>
    R2 Orange:LinearColor=<rgba 1.5 0.2458 0.0234 1>
    R2 Pink:LinearColor=<rgba 2 0.3958 1.672 1>
    R2 Purple:LinearColor=<rgba 0.2645 0.1302 1.25 1>
    R2 Saffron:LinearColor=<rgba 2 1.5503 0.0729 1>
    R2 SkyBlue:LinearColor=<rgba 0 1.3308 2 1>
    R2 TitaniumWhite:LinearColor=<rgba 2 2 2 1>
    Use R1 Painted Overrides=true
    Use R2 Painted Overrides=true
   )
   &VC_InitializeRibbon_v1(
    Color=$Particles.Color
    Custom Offset=true
    Enable Skeletal Position=true
    Local Vector Offset=&AddVectorToPosition(
     Position=/VehicleCosmetics/FX/Niagara/Modules/GameMode/VC_SetVectorByGameModeID(
      SameValueForLockerAndBR?:NiagaraBool=false
      Locker Vector:Vector3f=<-75 0 0>
      Battle Royale Vector:Vector3f=<120 0 0>
      Rocket Racing Vector:Vector3f=<250 0 0>
      GameModeID:NiagaraInt32=$System.GameModeID
     )
     Vector=$User.DTCarOffset
    )
    Offset Amount=1
    Ribbon Width=37
   )
   &AddVelocity(Velocity=&VC_SetVectorByGameModeID(
     SameValueForLockerAndBR?:NiagaraBool=false
     Locker Vector:Vector3f=<-2250 0 0>
     Battle Royale Vector:Vector3f=<-1000 0 0>
     Rocket Racing Vector:Vector3f=<-1000 0 0>
     GameModeID:NiagaraInt32=$System.GameModeID
    )) {module=AddVelocity001}
  ]
  particleUpdateScript [
   &ParticleState()
   &InheritSourceMovement(ApplyForce=false, ApplyPosition=true, ApplyPositionFalloff=false)
   &SolveForcesAndVelocity(Clamp Velocity=false, Limit Acceleration=false)
   &VC_UpdateRibbon_v1(Enable Skeletal Position=true)
   &DynamicMaterialParameters(
    Index 0 Param 1=&FloatFromCurve(FloatCurve=DataInterface() {propertyValues {Curve {keys [<at 0.5 0>,<at 1 1>]}}}, Scale Curve=0.5)
    Index 1 Param 1=5
    Index 1 Param 2=1.5
    Module.Index0Float0Write=true
    Module.Index0Float1Write=true
    Module.Index0Float2Write=true
    Module.Index0Float3Write=true
    Module.Index1Float0Write=true
    Module.Index1Float1Write=true
    Module.Index1Float2Write=true
    Module.Index1Float3Write=true
    Write Parameter Index 1=::NewEnumerator1
   )
   /SetVariables(Particles.SCALER1COLOR:LinearColor=&VC_DynamicScaleColor(
     Color to Scale:LinearColor=$Particles.Initial.R1 Color
     RGB Scale:Vector3f=<1 1 1>
     Alpha Scale:NiagaraFloat=&FloatFromCurve(FloatCurve=DataInterface() {propertyValues {Curve {defaultValue=1, keys [<at 0 0 RCIM_Cubic>,<at 0.2 1 RCIM_Cubic>]}}})
    )) {isSetParameters=true}
  ]
  rendererClasses [&NiagaraRibbonRendererProperties]
  renderers [
   {rendererClass=&NiagaraRibbonRendererProperties}
   {rendererClass=&NiagaraRibbonRendererProperties, rendererIndex=1}
   {rendererClass=&NiagaraRibbonRendererProperties, rendererIndex=2}
  ]
 }
 NE_VC_Boost_Sprite_Parent {
  emitterUpdateScript [
   &EmitterState()
   &VC_FilteredSocketCount()
   &SpawnRate(Spawn Probability=0.8, SpawnRate=10, Use Spawn Probability=true)
   &SpawnBurst_Instantaneous(Spawn Count=4, Spawn Group=1, Use Spawn Probability=false, UseLoopCountLimit=false)
   &SpawnBurst_Instantaneous(Spawn Count=4, Spawn Group=2, Use Spawn Probability=false, UseLoopCountLimit=false) {module=SpawnBurst_Instantaneous001}
  ]
  particleSpawnScript [
   &VC_PaintedVariantsOverride(
    Black=<rgba 0.5 0.5 0.5 0.0348>
    BurntSienna=<rgba 5 1.5438 0.6163 0.0609>
    Cobalt=<rgba 0 0.2775 15 0.4522>
    Crimson=<rgba 0.25 0 0.0042 1>
    DefaultColor=<rgba 45 18.75 0 0.03>
    ForestGreen=<rgba 0.1667 1 0 0.191>
    Gold:LinearColor=<rgba 1 1 1 0.2>
    Grey=<rgba 1 1 1 0.1>
    Lime=<rgba 0.35 0.7 0 1>
    Orange=<rgba 5 1.25 0 0.1>
    Pink=<rgba 25 2.9948 16.9981 0.0174>
    Purple=<rgba 8.2083 0 20 0.03>
    R1 Black:LinearColor=<rgba 0.035 0.035 0.035 2>
    R1 BurntSienna:LinearColor=<rgba 0.15 0.0463 0.0185 1>
    R1 Crimson:LinearColor=<rgba 0.25 0 0 1>
    R1 DefaultColor=<rgba 1.5 0.825 0.15 0.75>
    R1 ForestGreen:LinearColor=<rgba 0.0833 0.5 0 0.6>
    R1 Grey:LinearColor=<rgba 0.1 0.1 0.1 2>
    R1 Lime:LinearColor=<rgba 0.5 1 0 0.5>
    R1 Orange:LinearColor=<rgba 1 0.2083 0.05 0.4>
    R1 Pink:LinearColor=<rgba 1 0.3 0.7455 0.7>
    R1 Purple:LinearColor=<rgba 0.2265 0 0.6795 1.2>
    R1 Saffron:LinearColor=<rgba 1 0.7783 0.05 0.4>
    R1 SkyBlue:LinearColor=<rgba 0.075 0.5241 0.75 0.5>
    R1 TitaniumWhite:LinearColor=<rgba 0.75 0.75 0.75 0.4>
    R2 Black:LinearColor=<rgba 0.0833 0.0833 0.0833 2>
    R2 BurntSienna:LinearColor=<rgba 0.4792 0.1348 0.0424 1>
    R2 Cobalt:LinearColor=<rgba 0 0.4625 25 0.3652>
    R2 Crimson:LinearColor=<rgba 0.8 0 0 3>
    R2 DefaultColor:LinearColor=<rgba 2 0.4167 0.1 1>
    R2 ForestGreen:LinearColor=<rgba 0 0.75 0.0289 2>
    R2 Grey:LinearColor=<rgba 0.25 0.25 0.25 2>
    R2 Lime:LinearColor=<rgba 0.5833 1 0 2>
    R2 Orange:LinearColor=<rgba 1 0.25 0 2>
    R2 Pink:LinearColor=<rgba 2 0.4 1.4182 1>
    R2 Purple:LinearColor=<rgba 0.7482 0 1.8229 1>
    R2 Saffron:LinearColor=<rgba 1 0.9005 0.0365 1>
    R2 SkyBlue:LinearColor=<rgba 0 0.6654 1 0.5739>
    R2 TitaniumWhite:LinearColor=<rgba 0.9271 0.9271 0.9271 1>
    Saffron=<rgba 5 3.8333 0 0.0435>
    SkyBlue=<rgba 0 16.6346 25 0.012>
    TitaniumWhite=<rgba 5 5 5 0.0261>
    Use R1 Painted Overrides=true
    Use R2 Painted Overrides=true
   )
   &InitializeParticle(
    Color=$Particles.Color
    Lifetime=/VehicleCosmetics/FX/Niagara/Modules/MergeEmitters/VC_FloatBySpawnGroup(
     SpawnGroup or ExecIndex:NiagaraInt32=0
     CustomInt:NiagaraInt32=0
     MaximumSpawnGroups:NiagaraInt32=3
     Float SG 0:NiagaraFloat=&RandomRangeFloat(Minimum=0.25, Maximum=0.5)
     Float SG 1:NiagaraFloat=$Emitter.CurrentLoopDuration
     Float SG 2:NiagaraFloat=$Emitter.CurrentLoopDuration
     Float SG 3:NiagaraFloat=0
     Float SG 4:NiagaraFloat=0
     Float SG 5:NiagaraFloat=0
    )
    Mesh Scale Mode=::NewEnumerator3
    Mesh Uniform Scale=0.18
    Sprite Rotation Mode=::NewEnumerator2
    Sprite Size Mode=::NewEnumerator3
    Uniform Sprite Size=&MakeCustomFloatFromBool(Bool=$User.bIsPainted, True Float=80, False Float=65)
    UsePositionOffset=false
    Write Material Random=true
   )
   &PartitionParticles(Number Of Partitions=$Emitter.PortCount, Particles Per Partition=1, Partition Mode:ENiagara_PartitionMode=::NewEnumerator1)
   &SkeletalMeshLocation(
    Bone / Socket Index=$Particles.PartitionParticles.Partition
    Bone / Socket Sampling Mode=::NewEnumerator5
    MaskSpawnGroup=false
    Mesh Sampling Type=::NewEnumerator6
    Sampled Position Offset=&AddVectorToPosition(Position=<-75 0 0>, Vector=$User.DTCarOffset)
    Skeletal Mesh=$System.SkeletalMesh
   )
   &AddVelocity(Velocity=<0 0 0>)
   &InitialMeshOrientation(Mesh Axis To Orient=<1 1 1>, Mesh Orientation Mode=::NewEnumerator2, Use Rotation Vector=false)
  ]
  particleUpdateScript [
   /SetVariables(
    Particles.CenterSpriteSize:Vector2f=&Multiply_Vector2D_ByFloat(
     Float=&FloatFromCurve(FloatCurve=DataInterface() {propertyValues {Curve {keys [{tangentMode=RCTM_User, time=0, value=0}, <at 0.2 1>, <at 0.4 0.25>, <at 0.6 1>, <at 0.8 0.25>, {tangentMode=RCTM_User, time=1, value=0}]}}})
     Vector2D=&RandomRangeVector2D(Minimum=<80 275>, Maximum=<100 100>)
    )
    Particles.QTagVis:NiagaraInt32=&SpawnGroup()
   ) {isSetParameters=true, module=SetVariables_#2}
   &ParticleState()
   &UpdateMeshOrientation(
    Rotation Rate=&SelectFloatFromArray(
     Array Sampling Mode=::NewEnumerator2
     Direct Array Index=?
     Interpolation=&SelectFloatFromArray(
      Array Sampling Mode=::NewEnumerator0
      Direct Array Index=?
      Float Selection Array=DataInterface() {propertyValues {FloatData [0,1]}}
      Random Seed=0
     )
     Float Selection Array=DataInterface() {propertyValues {FloatData [-0.3,0.3]}}
    )
    Rotation Vector=&SelectVectorFromArray(
     Array Sampling Mode=::NewEnumerator2
     Direct Array Index=?
     Interpolation=&SelectFloatFromArray(
      Array Sampling Mode=::NewEnumerator0
      Direct Array Index=?
      Float Selection Array=DataInterface() {propertyValues {FloatData [0,1]}}
      Random Seed=0
     )
     Vector Selection Array=DataInterface() {propertyValues {FloatData [<-1 -1 -1>, <1 1 1>]}}
    )
   )
   &SolveForcesAndVelocity(Clamp Velocity=false, Limit Acceleration=false)
   &DynamicMaterialParameters(
    Index 0 Param 1=2
    Index 0 Param 2=&SelectFloatFromArray(
     Array Sampling Mode=::NewEnumerator0
     Direct Array Index=?
     Float Selection Array=DataInterface() {propertyValues {FloatData [2,3]}}
     Random Seed=0
    )
    Index 0 Param 3=3
    Module.Index0Float0Write=true
    Module.Index0Float1Write=true
    Module.Index0Float2Write=true
    Module.Index0Float3Write=true
   )
   &ScaleColor(
    Scale Alpha=&FloatFromCurve(FloatCurve=DataInterface() {propertyValues {Curve {keys [<at 0 1>,<at 1 0.5>]}}})
    Scale RGB=&VectorFromFloat(Value=&FloatFromCurve(
      FloatCurve=DataInterface() {propertyValues {Curve {keys [{arriveTangent=-1.9431, interpMode=RCIM_Cubic, leaveTangent=-1.9431, tangentMode=RCTM_User, time=0, value=1}, <at 1 0 RCIM_Cubic>]}}}
      Scale Curve=10
     ))
    ScaleA=true
    ScaleRGB=true
   ) {module=ScaleColor001}
  ]
  rendererClasses [&NiagaraSpriteRendererProperties]
  renderers [{rendererClass=&NiagaraSpriteRendererProperties}]
 }
 NE_VC_Boost_Sprite_Parent002 {
  emitterUpdateScript [
   &EmitterState(
    Inactive Response=::NewEnumerator1
    Life Cycle Mode=::NewEnumerator1
    Loop Behavior=::NewEnumerator1
    Loop Duration Mode=::NewEnumerator1
    UseLoopDelay=false
   )
   &VC_FilteredSocketCount()
   &SpawnRate(Spawn Probability=0.8, SpawnRate=10, Use Spawn Probability=true)
   &SpawnBurst_Instantaneous(Spawn Count=4, Spawn Group=1, Use Spawn Probability=false, UseLoopCountLimit=false)
   &SpawnBurst_Instantaneous(Spawn Count=4, Spawn Group=2, Use Spawn Probability=false, UseLoopCountLimit=false) {module=SpawnBurst_Instantaneous001}
  ]
  particleSpawnScript [
   /SetVariables(Particles.QTagVis:NiagaraInt32=&SpawnGroup()) {isSetParameters=true}
   &VC_PaintedVariantsOverride(
    Black=<rgba 0.5 0.5 0.5 0.0348>
    BurntSienna=<rgba 5 1.5438 0.6163 0.0609>
    Cobalt=<rgba 0 0.2775 15 0.4522>
    Crimson=<rgba 0.25 0 0.0042 1>
    DefaultColor=<rgba 45 18.75 0 0.03>
    ForestGreen=<rgba 0.1667 1 0 0.191>
    Gold:LinearColor=<rgba 1 1 1 0.2>
    Grey=<rgba 1 1 1 0.1>
    Lime=<rgba 0.35 0.7 0 1>
    Orange=<rgba 5 1.25 0 0.1>
    Pink=<rgba 25 2.9948 16.9981 0.0174>
    Purple=<rgba 8.2083 0 20 0.03>
    R1 Black:LinearColor=<rgba 0.035 0.035 0.035 2>
    R1 BurntSienna:LinearColor=<rgba 0.15 0.0463 0.0185 1>
    R1 Crimson:LinearColor=<rgba 0.25 0 0 1>
    R1 DefaultColor=<rgba 1.5 0.825 0.15 0.75>
    R1 ForestGreen:LinearColor=<rgba 0.0833 0.5 0 0.6>
    R1 Grey:LinearColor=<rgba 0.1 0.1 0.1 2>
    R1 Lime:LinearColor=<rgba 0.5 1 0 0.5>
    R1 Orange:LinearColor=<rgba 1 0.2083 0.05 0.4>
    R1 Pink:LinearColor=<rgba 1 0.3 0.7455 0.7>
    R1 Purple:LinearColor=<rgba 0.2265 0 0.6795 1.2>
    R1 Saffron:LinearColor=<rgba 1 0.7783 0.05 0.4>
    R1 SkyBlue:LinearColor=<rgba 0.075 0.5241 0.75 0.5>
    R1 TitaniumWhite:LinearColor=<rgba 0.75 0.75 0.75 0.4>
    R2 Black:LinearColor=<rgba 0.0833 0.0833 0.0833 2>
    R2 BurntSienna:LinearColor=<rgba 0.4792 0.1348 0.0424 1>
    R2 Cobalt:LinearColor=<rgba 0 0.4625 25 0.3652>
    R2 Crimson:LinearColor=<rgba 0.8 0 0 3>
    R2 DefaultColor:LinearColor=<rgba 2 0.4167 0.1 1>
    R2 ForestGreen:LinearColor=<rgba 0 0.75 0.0289 2>
    R2 Grey:LinearColor=<rgba 0.25 0.25 0.25 2>
    R2 Lime:LinearColor=<rgba 0.5833 1 0 2>
    R2 Orange:LinearColor=<rgba 1 0.25 0 2>
    R2 Pink:LinearColor=<rgba 2 0.4 1.4182 1>
    R2 Purple:LinearColor=<rgba 0.7482 0 1.8229 1>
    R2 Saffron:LinearColor=<rgba 1 0.9005 0.0365 1>
    R2 SkyBlue:LinearColor=<rgba 0 0.6654 1 0.5739>
    R2 TitaniumWhite:LinearColor=<rgba 0.9271 0.9271 0.9271 1>
    Saffron=<rgba 5 3.8333 0 0.0435>
    SkyBlue=<rgba 0 16.6346 25 0.012>
    TitaniumWhite=<rgba 5 5 5 0.0261>
    Use R1 Painted Overrides=true
    Use R2 Painted Overrides=true
   )
   &InitializeParticle(
    Color=$Particles.Color
    Lifetime=&VC_FloatBySpawnGroup(
     SpawnGroup or ExecIndex:NiagaraInt32=0
     CustomInt:NiagaraInt32=0
     MaximumSpawnGroups:NiagaraInt32=3
     Float SG 0:NiagaraFloat=&RandomRangeFloat(Minimum=0.25, Maximum=0.5)
     Float SG 1:NiagaraFloat=$Emitter.CurrentLoopDuration
     Float SG 2:NiagaraFloat=$Emitter.CurrentLoopDuration
     Float SG 3:NiagaraFloat=0
     Float SG 4:NiagaraFloat=0
     Float SG 5:NiagaraFloat=0
    )
    Mesh Scale Mode=::NewEnumerator3
    Mesh Uniform Scale=0.18
    Sprite Rotation Mode=::NewEnumerator2
    Sprite Size Mode=::NewEnumerator3
    Uniform Sprite Size=&MakeCustomFloatFromBool(Bool=$User.bIsPainted, True Float=80, False Float=65)
    UsePositionOffset=false
    Write Material Random=true
   )
   &PartitionParticles(Number Of Partitions=$Emitter.PortCount, Particles Per Partition=1, Partition Mode:ENiagara_PartitionMode=::NewEnumerator1)
   &SkeletalMeshLocation(
    Bone / Socket Index=$Particles.PartitionParticles.Partition
    Bone / Socket Sampling Mode=::NewEnumerator5
    MaskSpawnGroup=false
    Mesh Sampling Type=::NewEnumerator6
    Sampled Position Offset=&AddVectorToPosition(Position=<-75 0 0>, Vector=$User.DTCarOffset)
    Skeletal Mesh=$System.SkeletalMesh
   )
   &AddVelocity(Velocity=<0 0 0>)
   &InitialMeshOrientation(Mesh Axis To Orient=<1 1 1>, Mesh Orientation Mode=::NewEnumerator2, Use Rotation Vector=false)
  ]
  particleUpdateScript [
   /SetVariables(Particles.CenterSpriteSize:Vector2f=&Multiply_Vector2D_ByFloat(
     Float=&FloatFromCurve(FloatCurve=DataInterface() {propertyValues {Curve {keys [{tangentMode=RCTM_User, time=0, value=0}, <at 0.2 1>, <at 0.4 0.25>, <at 0.6 1>, <at 0.8 0.25>, {tangentMode=RCTM_User, time=1, value=0}]}}})
     Vector2D=&RandomRangeVector2D(Minimum=<80 275>, Maximum=<100 100>)
    )) {isSetParameters=true, module=SetVariables_#2}
   &ParticleState(Kill Particles When Lifetime Has Elapsed=false, Loop Particles Lifetime=true)
   &UpdateMeshOrientation(
    Rotation Rate=&SelectFloatFromArray(
     Array Sampling Mode=::NewEnumerator2
     Direct Array Index=?
     Interpolation=&SelectFloatFromArray(
      Array Sampling Mode=::NewEnumerator0
      Direct Array Index=?
      Float Selection Array=DataInterface() {propertyValues {FloatData [0,1]}}
      Random Seed=0
     )
     Float Selection Array=DataInterface() {propertyValues {FloatData [-0.3,0.3]}}
    )
    Rotation Vector=&SelectVectorFromArray(
     Array Sampling Mode=::NewEnumerator2
     Direct Array Index=?
     Interpolation=&SelectFloatFromArray(
      Array Sampling Mode=::NewEnumerator0
      Direct Array Index=?
      Float Selection Array=DataInterface() {propertyValues {FloatData [0,1]}}
      Random Seed=0
     )
     Vector Selection Array=DataInterface() {propertyValues {FloatData [<-1 -1 -1>, <1 1 1>]}}
    )
   )
   &SolveForcesAndVelocity(Clamp Velocity=false, Limit Acceleration=false)
   &DynamicMaterialParameters(
    Index 0 Param 1=1000
    Index 0 Param 2=&SelectFloatFromArray(
     Array Sampling Mode=::NewEnumerator0
     Direct Array Index=?
     Float Selection Array=DataInterface() {propertyValues {FloatData [1,5]}}
     Random Seed=0
    )
    Module.Index0Float0Write=true
    Module.Index0Float1Write=true
    Module.Index0Float2Write=true
    Module.Index0Float3Write=true
   )
   &ScaleColor(
    Scale Alpha=&FloatFromCurve(FloatCurve=DataInterface() {propertyValues {Curve {keys [<at 0 1>,<at 1 0.5>]}}})
    Scale RGB=&VectorFromFloat(Value=&FloatFromCurve(
      FloatCurve=DataInterface() {propertyValues {Curve {keys [{arriveTangent=-1.9431, interpMode=RCIM_Cubic, leaveTangent=-1.9431, tangentMode=RCTM_User, time=0, value=1}, <at 1 0 RCIM_Cubic>]}}}
      Scale Curve=10
     ))
    ScaleA=true
    ScaleRGB=true
   ) {module=ScaleColor001}
  ]
  rendererClasses [&NiagaraSpriteRendererProperties, &NiagaraMeshRendererProperties]
  renderers [
   {rendererClass=&NiagaraSpriteRendererProperties}
   {rendererClass=&NiagaraMeshRendererProperties, rendererIndex=1}
   {rendererClass=&NiagaraSpriteRendererProperties, rendererIndex=2}
  ]
 }
 NE_VC_Boost_Sprite_Parent003 {
  emitterUpdateScript [
   &EmitterState()
   &VC_FilteredSocketCount()
   &SpawnRate(SpawnRate=15, Use Spawn Probability=false)
   &SpawnRate(SpawnGroup=1, SpawnRate=10, Use Spawn Probability=false) {module=SpawnRate001}
   &SpawnRate(SpawnGroup=2, SpawnRate=30, Use Spawn Probability=false) {module=SpawnRate002}
  ]
  particleSpawnScript [
   &VC_PaintedVariantsOverride(
    Black=<rgba 0.0677 0.0677 0.0677 1>
    BurntSienna=<rgba 0.5156 0.1592 0.0636 1>
    Cobalt=<rgba 0 0.925 50 0.087>
    Crimson=<rgba 0.5 0 0.0136 2>
    DefaultColor=<rgba 1.5 0.75 0.045 1>
    ForestGreen=<rgba 0.1667 1 0 1>
    Grey=<rgba 0.5 0.5 0.5 0.6783>
    Lime=<rgba 0.5 1 0 1>
    Orange=<rgba 2 0.6976 0 1>
    Pink=<rgba 5 0.599 3.3996 1>
    Purple=<rgba 1.2312 0 3 1>
    R1 Black:LinearColor=<rgba 0.2696 0.2696 0.2696 0.6261>
    R1 BurntSienna:LinearColor=<rgba 1 0.3088 0.1233 1>
    R1 Cobalt:LinearColor=<rgba 0 0.925 50 0.087>
    R1 Crimson:LinearColor=<rgba 1 0 0.0271 2>
    R1 DefaultColor=<rgba 5 2.0833 0 0.5>
    R1 ForestGreen:LinearColor=<rgba 0.1667 1 0 1>
    R1 Grey:LinearColor=<rgba 1 1 1 0.6783>
    R1 Lime:LinearColor=<rgba 0.5 1 0 1>
    R1 Orange:LinearColor=<rgba 2 0.5 0 1>
    R1 Pink:LinearColor=<rgba 5 0.599 3.3996 1>
    R1 Purple:LinearColor=<rgba 1.2312 0 3 1>
    R1 Saffron:LinearColor=<rgba 5 3.8333 0 0.487>
    R1 SkyBlue:LinearColor=<rgba 0 3.3269 5 0.4261>
    R1 TitaniumWhite:LinearColor=<rgba 5 5 5 0.4783>
    R2 Black:LinearColor=<rgba 0.6609 0.6609 0.6609 0.1739>
    R2 BurntSienna:LinearColor=<rgba 2 0.3667 0.04 1>
    R2 Cobalt:LinearColor=<rgba 0 0.925 50 0.5>
    R2 Crimson:LinearColor=<rgba 0.8 0 0.0217 1>
    R2 DefaultColor:LinearColor=<rgba 1 0 0 10>
    R2 ForestGreen:LinearColor=<rgba 0 0.5 0.0253 1>
    R2 Grey:LinearColor=<rgba 1 1 1 0.6783>
    R2 Lime:LinearColor=<rgba 0.5 1 0 1>
    R2 Orange:LinearColor=<rgba 2 0.6976 0 0.5>
    R2 Pink:LinearColor=<rgba 5 0.599 3.3996 1>
    R2 Purple:LinearColor=<rgba 1.2312 0 3 1>
    R2 Saffron:LinearColor=<rgba 5 3.8333 0 0.6>
    R2 SkyBlue:LinearColor=<rgba 0 3.3269 5 0.3304>
    R2 TitaniumWhite:LinearColor=<rgba 2 2 2 0.513>
    Saffron=<rgba 5 3.8333 0 1>
    SkyBlue=<rgba 0 3.3269 5 1>
    TitaniumWhite=<rgba 2 2 2 0.4609>
    Use R1 Painted Overrides=true
    Use R2 Painted Overrides=true
   )
   &InitializeParticle(
    Color=$Particles.Color
    Lifetime=&VC_FloatBySpawnGroup(
     SpawnGroup or ExecIndex:NiagaraInt32=0
     CustomInt:NiagaraInt32=0
     MaximumSpawnGroups:NiagaraInt32=3
     Float SG 0:NiagaraFloat=&RandomRangeFloat(Minimum=0.5, Maximum=0.75)
     Float SG 1:NiagaraFloat=&RandomRangeFloat(Minimum=0.25, Maximum=0.5)
     Float SG 2:NiagaraFloat=&RandomRangeFloat(Minimum=0.3, Maximum=0.4)
     Float SG 3:NiagaraFloat=0
     Float SG 4:NiagaraFloat=0
     Float SG 5:NiagaraFloat=0
    )
    Sprite Rotation Mode=::NewEnumerator2
    Sprite Size Mode=::NewEnumerator3
    Uniform Sprite Size=&VC_FloatBySpawnGroup(
     SpawnGroup or ExecIndex:NiagaraInt32=0
     CustomInt:NiagaraInt32=0
     MaximumSpawnGroups:NiagaraInt32=3
     Float SG 0:NiagaraFloat=35
     Float SG 1:NiagaraFloat=&RandomRangeFloat(Minimum=25, Maximum=35)
     Float SG 2:NiagaraFloat=&RandomRangeFloat(Minimum=25, Maximum=25)
     Float SG 3:NiagaraFloat=0
     Float SG 4:NiagaraFloat=0
     Float SG 5:NiagaraFloat=0
    )
    UsePositionOffset=false
    Write Material Random=true
   )
   &PartitionParticles(Number Of Partitions=$Emitter.PortCount)
   &SkeletalMeshLocation(
    Bone / Socket Index=$Particles.PartitionParticles.Partition
    Bone / Socket Sampling Mode=::NewEnumerator5
    MaskSpawnGroup=false
    Mesh Sampling Type=::NewEnumerator6
    Sampled Position Offset=&AddVectorToPosition(Position=<-75 0 0>, Vector=$System.FinalPosition)
    Skeletal Mesh=$System.SkeletalMesh
   )
   &AddVelocity(Velocity=/VehicleCosmetics/FX/Niagara/Modules/MergeEmitters/VC_VectorBySpawnGroup(
     SpawnGroup or ExecIndex:NiagaraInt32=0
     CustomInt:NiagaraInt32=0
     MaximumSpawnGroups:NiagaraInt32=3
     Vector Group 0:Vector3f=<-1000 0 0>
     Vector Group 1:Vector3f=<0 0 0>
     Vector Group 2:Vector3f=<0 0 0>
     Vector Group 3:Vector3f=<0 0 0>
     Vector Group 4:Vector3f=<0 0 0>
     Vector Group 5:Vector3f=<0 0 0>
    ))
   &ShapeLocation(
    MaskSpawnGroup=false
    Non Uniform Scale=&VC_VectorBySpawnGroup(
     SpawnGroup or ExecIndex:NiagaraInt32=0
     CustomInt:NiagaraInt32=0
     MaximumSpawnGroups:NiagaraInt32=3
     Vector Group 0:Vector3f=<1 1 1>
     Vector Group 1:Vector3f=<1 1.5 1.5>
     Vector Group 2:Vector3f=<0.4 0.4 0.4>
     Vector Group 3:Vector3f=<0 0 0>
     Vector Group 4:Vector3f=<0 0 0>
     Vector Group 5:Vector3f=<0 0 0>
    )
    Override Seed=false
    Sphere Radius=&VC_FloatBySpawnGroup(
     SpawnGroup or ExecIndex:NiagaraInt32=0
     CustomInt:NiagaraInt32=0
     MaximumSpawnGroups:NiagaraInt32=3
     Float SG 0:NiagaraFloat=20
     Float SG 1:NiagaraFloat=&RandomRangeFloat(Minimum=35, Maximum=55)
     Float SG 2:NiagaraFloat=&RandomRangeFloat(Minimum=35, Maximum=55)
     Float SG 3:NiagaraFloat=25
     Float SG 4:NiagaraFloat=0
     Float SG 5:NiagaraFloat=0
    )
    UseOwnerScale=true
   )
  ]
  particleUpdateScript [
   /SetVariables(Particles.SKTagVis:NiagaraInt32=&SpawnGroup()) {isSetParameters=true}
   &ParticleState()
   &ScaleSpriteSize(Uniform Curve Sprite Scale=DataInterface() {propertyValues {Curve {keys [{tangentMode=RCTM_User, time=0, value=0}, <at 0.2 1>, <at 0.4 0.25>, <at 0.6 1>, <at 0.8 0.25>, {tangentMode=RCTM_User, time=1, value=0}]}}})
   &SolveForcesAndVelocity(Clamp Velocity=false, Limit Acceleration=false)
   &DynamicMaterialParameters(
    Index 0 Param 1=2
    Index 0 Param 2=&SelectFloatFromArray(
     Array Sampling Mode=::NewEnumerator0
     Direct Array Index=?
     Float Selection Array=DataInterface() {propertyValues {FloatData [2,3]}}
     Random Seed=0
    )
    Index 0 Param 3=3
    Module.Index0Float0Write=true
    Module.Index0Float1Write=true
    Module.Index0Float2Write=true
    Module.Index0Float3Write=true
   )
   &ScaleColor(
    Scale Alpha=&FloatFromCurve(FloatCurve=DataInterface() {propertyValues {Curve {
        defaultValue=1
        keys [
         {interpMode=RCIM_Cubic, tangentMode=RCTM_User, time=0, value=0}
         {interpMode=RCIM_Cubic, tangentMode=RCTM_User, time=0.5, value=1}
         {tangentMode=RCTM_User, time=1, value=0}
        ]
       }}})
    Scale RGB=&VectorFromFloat(Value=&FloatFromCurve(
      FloatCurve=DataInterface() {propertyValues {Curve {
         defaultValue=1
         keys [
          {interpMode=RCIM_Cubic, tangentMode=RCTM_User, time=0, value=0}
          {interpMode=RCIM_Cubic, tangentMode=RCTM_User, time=0.5, value=1}
          {tangentMode=RCTM_User, time=1, value=0}
         ]
        }}}
      Scale Curve=7.5
     ))
    ScaleA=true
    ScaleRGB=true
   ) {module=ScaleColor001}
  ]
  rendererClasses [&NiagaraSpriteRendererProperties]
  renderers [
   {rendererClass=&NiagaraSpriteRendererProperties}
   {rendererClass=&NiagaraSpriteRendererProperties, rendererIndex=1}
   {rendererClass=&NiagaraSpriteRendererProperties, rendererIndex=2}
  ]
 }
}
summary {
 systemName=NS_Boost_Modish_MP2
 emitters [
  {
   emitterName=NE_VC_Boost_Sprite_Parent002
   rendererClasses [&NiagaraSpriteRendererProperties, &NiagaraMeshRendererProperties]
  }
  {emitterName=NE_VC_Boost_Sprite_Parent003, rendererClasses [&NiagaraSpriteRendererProperties]}
  {emitterName=NE_VC_Boost_Ribbon_Parent001, rendererClasses [&NiagaraRibbonRendererProperties]}
  {emitterName=NE_VC_Boost_Sprite_Parent, rendererClasses [&NiagaraSpriteRendererProperties]}
 ]
 userVariables [
  User.bIsFrontendPreview {type=NiagaraBool}
  User.bIsPainted {type=NiagaraBool}
  User.ColorVariantID {type=NiagaraInt32}
  User.DTCarOffset {type=Vector3f, description="", defaultValue {struct=Vector3f, value=<0 0 0>}}
  User.PaintedColor {type=LinearColor, description="", defaultValue {struct=LinearColor, value=<rgba 1 1 1 1>}}
  User.PaintedScalar {type=NiagaraFloat, defaultValue=1, description=""}
  User.PortCount {type=NiagaraInt32, defaultValue=1}
 ]
}
