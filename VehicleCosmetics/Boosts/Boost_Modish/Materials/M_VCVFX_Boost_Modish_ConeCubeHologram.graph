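# Translucent unlit two-sided hologram material for the Modish boost VFX (ConeCubeHologram), usable on Niagara sprites, ribbons and mesh particles.
# Three GeneratedBand line patterns sample panned world-Z at separate parameterized spacings.
# The first is masked by a fixed narrow band, then added to a faint mix of itself and the product of the other two.
# Particle color tints the result through MF_FX_EmissiveAdjustments, and particle alpha scales a saturated copy.
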Material /VehicleCosmetics/Boosts/Boost_Modish/Materials/M_VCVFX_Boost_Modish_ConeCubeHologram
graph {
 properties {
  bUsedWithNiagaraMeshParticles=true
  bUsedWithNiagaraRibbons=true
  bUsedWithNiagaraSprites=true
  blendMode=BLEND_Translucent
  shadingModel=MSM_Unlit
  twoSided=true
  usedShadingModels=Unlit
 }
 nodes [
  n1 = Constant() {r=0.04}
  n2 = Constant() {r=1}
  n3 = Constant() {r=0.46}
  n4 = WorldPosition()
  n5 = Divide(A=@n4.XYZ, B=ScalarParameter() {defaultValue=10, parameterName=Spacing_1})
  n6 = MaterialFunctionCall!(
   Width=ScalarParameter() {defaultValue=0.7, parameterName=LineWidt_1}
   Sharpness
   Offset
   Direction Switch
   Compare
   Input Coordinates=Panner(Coordinate=ComponentMask(None=@n5) {b=true}) {speedY=2}
  ) {materialFunction=/Engine/Functions/Engine_MaterialFunctions03/Procedurals/GeneratedBand}
  n7 = ComponentMask(None=Divide(A=@n4.XYZ, B=ScalarParameter() {defaultValue=20, parameterName=Spacing_2})) {b=true}
  n8 = ComponentMask(None=Divide(A=@n4.XYZ, B=ScalarParameter() {defaultValue=3, parameterName=Spacing_3})) {b=true}
  n9 = Multiply(
   A=MaterialFunctionCall(Width=ScalarParameter() {parameterName=LineWidth_2}, Input Coordinates=Panner(Coordinate=@n7) {speedY=0.8}) {materialFunction=&GeneratedBand}.Result
   B=MaterialFunctionCall(Width=ScalarParameter() {defaultValue=0.5, parameterName=LineWidth_3}, Input Coordinates=Panner(Coordinate=@n8) {speedY=2}) {materialFunction=&GeneratedBand}.Result
  )
  n10 = Add(
   A=Multiply(
    A=MaterialFunctionCall(Width=@n1, Sharpness=@n2, Offset=@n3) {materialFunction=&GeneratedBand}.Result
    B=@n6.Result
   )
   B=Multiply(A=Add(A=@n6.Result, B=@n9), B=Constant() {r=0.05})
  )
  MaterialExpressionMeshEmitterVertexColor_0 = ParticleColor()
  MaterialExpressionConstantClamp_0 = Saturate(None=Multiply(A=@MaterialExpressionMeshEmitterVertexColor_0.A, B=@n10))
  MaterialFunctionCall(Emissive=Multiply(A=@MaterialExpressionMeshEmitterVertexColor_0.RGB, B=@n10)) {materialFunction=&MF_FX_EmissiveAdjustments}
 ]
}
