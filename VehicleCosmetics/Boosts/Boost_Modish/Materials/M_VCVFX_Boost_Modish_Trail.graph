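# Translucent unlit Niagara material for the Modish vehicle boost trail, usable on ribbons, sprites and mesh particles.
# Color comes from a noise-distorted, panning RGB dissolve texture, tinting each channel red or orange before desaturation and particle color.
# Opacity multiplies sharpened, distorted Perlin noise by a floored, hard-stepped noise mask shaped along U, then particle alpha and depth fade.
# An XOffset dynamic parameter, shared through a named reroute, shifts the UVs.
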Material /VehicleCosmetics/Boosts/Boost_Modish/Materials/M_VCVFX_Boost_Modish_Trail
graph {
 properties {
  bAutomaticallySetUsageInEditor=false
  bUsedWithNiagaraMeshParticles=true
  bUsedWithNiagaraRibbons=true
  bUsedWithNiagaraSprites=true
  blendMode=BLEND_Translucent
  shadingModel=MSM_Unlit
  usedShadingModels=Unlit
 }
 nodes [
  n1 = Add(A=TextureCoordinate() {uTiling=2, vTiling=0.1}, B=AppendVector(A=NamedRerouteUsage(), B=Constant()))
  n2 = NamedRerouteUsage()
  n3 = Constant()
  n4 = Panner(Coordinate=Add(A=TextureCoordinate() {uTiling=1.5, vTiling=0.2}, B=AppendVector(A=@n2, B=@n3))) {speedX=-1, speedY=0.6}
  n5 = Add(
   A=ComponentMask(None=TextureSample(UVs=@n4) {texture=/DealershipSharedLibrary/FX/Textures/Noise/T_VCVFX_Boost_S16_PerlinBlur}.RGB) {g=true, r=true}
   B=AppendVector(A=Constant() {r=0.85}, B=Constant() {r=0.1})
  )
  n6 = TextureSample(UVs=Panner(Coordinate=Multiply(A=@n1, B=@n5)) {speedX=-1, speedY=-0.15}) {texture=&T_VCVFX_Boost_S16_PerlinBlur}
  n7 = Constant() {r=9}
  MaterialExpressionConstantClamp_4 = Saturate(None=Multiply(A=Power(Base=@n6.RGB, Exp=@n7), B=Constant() {r=50000}))
  n8 = Add(
   A=TextureCoordinate() {uTiling=6, vTiling=0.1}
   B=MaterialFunctionCall(X=NamedRerouteUsage()) {materialFunction=&MakeFloat2}.Result
  )
  n9 = OneMinus(None=ComponentMask(None=TextureCoordinate()) {r=true})
  n10 = Constant() {r=13}
  n11 = Multiply(
   A=TextureSample(UVs=Panner(Coordinate=@n8) {speedX=-1, speedY=-0.3}) {texture=&T_VCVFX_Boost_S16_PerlinBlur}.RGB
   B=Power(Base=Multiply(A=@n9, B=@n10), Exp=Constant() {r=3})
  )
  n12 = ComponentMask(None=TextureCoordinate()) {r=true}
  n13 = Multiply(A=@MaterialExpressionConstantClamp_4, B=Multiply(A=Floor(None=@n11), B=Power(Base=@n12) {constExponent=4}))
  n14 = Constant() {r=90}
  MaterialExpressionVertexColor_1 = ParticleColor()
  MaterialExpressionConstantClamp_3 = Saturate(None=Multiply(A=Multiply(A=@n13, B=@n14), B=@MaterialExpressionVertexColor_1.A))
  MaterialFunctionCall(Opacity In=@MaterialExpressionConstantClamp_3, DepthFade Distance=Constant() {r=25}) {materialFunction=&MF_VCVFX_DepthFade}
  n15 = Constant3Vector() {constant=<rgba 1 0 0 0>}
  n16 = Add(
   A=TextureCoordinate() {uTiling=1.5, vTiling=0}
   B=MaterialFunctionCall(X=NamedRerouteUsage()) {materialFunction=&MakeFloat2}.Result
  )
  n17 = Panner(Coordinate=TextureCoordinate() {uTiling=3, vTiling=0.25}) {speedX=-0.5, speedY=0.3}
  n18 = Multiply(
   A=ComponentMask(None=TextureSample(UVs=@n17) {texture=&T_VCVFX_Boost_S16_PerlinBlur}.RGB) {g=true, r=true}
   B=Constant() {r=0.4}
  )
  n19 = TextureSample(UVs=Panner(Coordinate=Add(A=@n16, B=@n18)) {speedX=-0.65, speedY=0.3}) {texture=/DealershipSharedLibrary/FX/Textures/TexturePackage/T_VCVFX_Modish_DissTrail2_RGB}
  n20 = Constant3Vector() {constant=<rgba 1 0.1 0.01 0>}
  n21 = Add(A=Add(A=Multiply(A=@n15, B=@n19.R), B=Multiply(A=@n19.G, B=@n20)), B=Multiply(A=Constant3Vector() {constant=<rgba 0.2 0 0 0>}, B=@n19.B))
  n22 = ScalarParameter() {parameterName=Desaturation}
  MaterialFunctionCall(Emissive=Multiply(A=@MaterialExpressionVertexColor_1.RGB, B=Desaturation(None=@n21, Fraction=@n22))) {materialFunction=&MF_FX_EmissiveAdjustments}
  n23 = DynamicParameter() {paramNames [Desaturation,XOffset,Param3,Param4]}
  NamedRerouteDeclaration(None=@n23.XOffset) {name=XOffset, nodeColor=<rgba 0.4531 1 0 1>}
  Reroute(None=@n23.XOffset)
 ]
}
