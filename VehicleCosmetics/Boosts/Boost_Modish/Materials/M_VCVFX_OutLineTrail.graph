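# Translucent unlit Niagara material for an outline trail effect in the Modish vehicle boost cosmetic.
# An outline texture is multiplied by a panning, squared dissolve texture and U-coordinate gradients, shaping opacity along the trail, which then passes through a depth fade function.
# Particle color tints the outline for emissive and its alpha scales opacity.
# Dynamic parameter index 1 drives brightness, leading-edge falloff power and dissolve UV offset.
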Material /VehicleCosmetics/Boosts/Boost_Modish/Materials/M_VCVFX_OutLineTrail
graph {
 properties {
  bAutomaticallySetUsageInEditor=false
  bUsedWithNiagaraMeshParticles=true
  bUsedWithNiagaraRibbons=true
  bUsedWithNiagaraSprites=true
  blendMode=BLEND_Translucent
  shadingModel=MSM_Unlit
  usedShadingModels=Unlit
 }
 nodes [
  n1 = DynamicParameter() {parameterIndex=1, paramNames ["Black&&WhiteControl", PowerInicial, Offset, Param4]}
  n2 = Power(Base=ComponentMask(None=TextureCoordinate()) {r=true}, Exp=@n1.PowerInicial)
  n3 = Constant() {r=20}
  n4 = OneMinus(None=ComponentMask(None=TextureCoordinate()) {r=true})
  n5 = AppendVector(A=@n1.Offset, B=Constant())
  n6 = TextureCoordinate()
  n7 = ScalarParameter() {defaultValue=3, parameterName=CrackTile_X}
  n8 = ScalarParameter() {defaultValue=1, parameterName=CrackTile_Y}
  n9 = Panner(Coordinate=Multiply(A=Add(A=@n5, B=@n6), B=AppendVector(A=@n7, B=@n8))) {speedX=-2}
  n10 = Multiply(
   A=TextureSample(UVs=@n9) {texture=/VehicleCosmetics/Boosts/Boost_Modish/Textures/T_VCVFX_DissolveTrails}.RGB
   B=TextureSample(UVs=@n9) {texture=&T_VCVFX_DissolveTrails}.RGB
  )
  n11 = TextureSample() {texture=/VehicleCosmetics/Boosts/Boost_Modish/Textures/T_VCVFX_OutineTrail}
  n12 = Add(A=ConstantBiasScale(None=@n4) {bias=-0.35, scale=1.1}, B=Multiply(A=Multiply(A=@n11.RGB, B=@n10), B=Constant() {r=3}))
  n13 = Power(Base=Constant() {r=0.5}, Exp=@n4)
  n14 = Multiply(A=@n11.RGB, B=Multiply(A=Add(A=@n4, B=@n12), B=OneMinus(None=@n13)))
  MaterialExpressionVertexColor_1 = ParticleColor()
  MaterialExpressionConstantClamp_1 = Saturate(None=Multiply(A=Multiply(A=@n2, B=@n3), B=Multiply(A=@n14, B=@MaterialExpressionVertexColor_1.A)))
  n15 = TextureCoordinate()
  n16 = Power(Base=OneMinus(None=ComponentMask(None=@n15) {r=true})) {constExponent=0.2}
  MaterialFunctionCall(Opacity In=Multiply(A=@MaterialExpressionConstantClamp_1, B=Round(None=@n16)), DepthFade Distance=Constant() {r=5}) {materialFunction=&MF_VCVFX_DepthFade}
  MaterialFunctionCall(Emissive=Multiply(A=@n11.RGB, B=Multiply(A=@MaterialExpressionVertexColor_1.RGB, B=@n1."Black&&WhiteControl"))) {materialFunction=&MF_FX_EmissiveAdjustments}
 ]
}
